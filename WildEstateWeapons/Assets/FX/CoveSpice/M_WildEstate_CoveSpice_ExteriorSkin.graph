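# Unlit exterior-skin effect material for CoveSpice in WildEstateWeapons, drawing a pulsing charge-colored rim glow on skeletal or clothed meshes.
# ChargePercent blends yellow to orange, while a sine timed from StartT pulses brightness inside a Fresnel rim.
# With depth test disabled, biased pixel depth against custom depth masks the rim, and a second mask adds scene-depth occlusion.
# A NormalWPO term scales vertex normals for a shell offset.
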Material /WildEstateWeapons/Assets/FX/CoveSpice/M_WildEstate_CoveSpice_ExteriorSkin
graph {
 properties {
  allowTranslucentCustomDepthWrites=true
  bDisableDepthTest=true
  bUsedWithClothing=true
  bUsedWithMeshDeformer=true
  bUsedWithNanite=true
  bUsedWithSkeletalMesh=true
  pixelDepthOffsetMode=PDOM_Legacy
  shadingModel=MSM_Unlit
  usedShadingModels=Unlit
 }
 nodes [
  n1 = LinearInterpolate(
   A=VectorParameter() {parameterName=Color A, defaultValue=<rgba 4.9077 5 0 1>}.RGB
   B=VectorParameter() {parameterName=Color B, defaultValue=<rgba 5 0.6698 0 1>}.RGB
   Alpha=ScalarParameter() {parameterName=ChargePercent}
  )
  n2 = ScalarParameter() {defaultValue=0.2, parameterName=sA}
  n3 = Time()
  n4 = ScalarParameter() {parameterName=StartT}
  n5 = Sine(None=Multiply(A=Subtract(A=@n3, B=@n4)) {constB=7})
  n6 = ScalarParameter() {defaultValue=0.3174, parameterName=ExponentIn}
  n7 = Multiply(
   A=Power(Base=Fresnel(ExponentIn=@n6), Exp=ScalarParameter() {defaultValue=15, parameterName=Exp}) {constExponent=3}
   B=ScalarParameter() {defaultValue=4, parameterName=B}
  )
  LinearInterpolate(B=LinearInterpolate(A=Multiply(A=@n1, B=@n2), B=@n1, Alpha=Power(Base=@n5) {constExponent=4}), Alpha=Clamp(None=@n7))
  Multiply(A=VertexNormalWS(), B=ScalarParameter() {defaultValue=1.5, parameterName=NormalWPO})
  n8 = Constant()
  n9 = If(
   A=Add(A=PixelDepth().R, B=ScalarParameter() {defaultValue=-0.2, parameterName=BiasSorting_Far})
   B=ComponentMask(None=SceneTexture() {sceneTextureId=PPI_CustomDepth}.Color) {r=true}
   "A > B"=@n8
   "A == B"=@n8
   "A < B"=Constant() {r=1}
  ) {equalsThreshold=0.1}
  n10 = OneMinus(None=Fresnel() {baseReflectFraction=0, exponent=6})
  n11 = SceneDepth()
  n12 = PixelDepth()
  n13 = Constant() {r=1}
  n14 = Constant()
  Multiply(
   A=@n9
   B=Multiply(A=Power(Base=@n10) {constExponent=3}, B=If(A=@n11.R, B=@n12.R, "A > B"=@n13, "A == B"=@n13, "A < B"=@n14) {equalsThreshold=0.1})
  )
  Multiply(A=@n9, B=@n7)
 ]
}
