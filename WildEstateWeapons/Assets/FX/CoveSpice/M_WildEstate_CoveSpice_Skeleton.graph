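# Translucent unlit skeleton material for the WildEstate CoveSpice effect, usable on skeletal and Niagara meshes, with depth test disabled.
# It fakes sun shading from a flattened normal map and sky light direction, blending shadow and time-of-day-dimmed bone colors.
# An Alpha parameter mixes a skull texture into that shading mask.
# Separate branches compute eye-adaptation compensation, camera-relative offset, a manual scene-depth comparison and an object-relative sphere mask.
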Material /WildEstateWeapons/Assets/FX/CoveSpice/M_WildEstate_CoveSpice_Skeleton
graph {
 properties {
  bDisableDepthTest=true
  bUsedWithClothing=true
  bUsedWithMeshDeformer=true
  bUsedWithNiagaraMeshParticles=true
  bUsedWithSkeletalMesh=true
  blendMode=BLEND_Translucent
  pixelDepthOffsetMode=PDOM_Legacy
  shadingModel=MSM_Unlit
  usedShadingModels=Unlit
 }
 nodes [
  n1 = VectorParameter() {group=Colors, parameterName=BoneColor, defaultValue=<rgba 0.7 0.7 0.7 1>}
  n2 = Constant() {r=1}
  n3 = MaterialFunctionCall(Value_Morning=@n2, Value_Day=@n2, Value_Evening=@n2, Value_Night=Constant() {r=0.6}) {materialFunction=&Scalar_Time_of_Day}
  n4 = SkyAtmosphereLightDirection()
  n5 = TextureSampleParameter2D() {
   group=Normals
   parameterName=Normal
   samplerType=SAMPLERTYPE_Normal
   texture=/Game/Characters/Player/Male/Medium/Bodies/M_MED_Keychain/Textures/T_KeyChain_Body_N
  }
  n6 = Transform(None=MaterialFunctionCall(Normal=@n5.RGB, Flatness=ScalarParameter() {defaultValue=0.5, group=Normals, parameterName=Normal Flatness}) {materialFunction=&FlattenNormal}.Result)
  n7 = Add(
   A=DotProduct(A=Normalize(VectorInput=@n4), B=Normalize(VectorInput=@n6))
   B=Multiply(A=ScalarParameter() {defaultValue=0.3, group=Shadows, parameterName=Shadow Offset}) {constB=1.05}
  )
  n8 = MaterialFunctionCall(In=Saturate(None=Add(A=@n7) {constB=0.5}), Contrast=ScalarParameter() {defaultValue=3, group=Shadows, parameterName=Shadow Contrast}) {materialFunction=&CheapContrast}
  n9 = Multiply(A=TextureSample() {texture=&/T_FX_WildEstate_CoveSpice_Skull}.RGB, B=@n8.Result)
  n10 = ScalarParameter() {parameterName=Alpha}
  n11 = LinearInterpolate(
   A=VectorParameter() {group=Colors, parameterName=DarkColor, defaultValue=<rgba 0.0023 0.0156 0 1>}.RGB
   B=LinearInterpolate(
    A=VectorParameter() {group=Colors, parameterName=ShadowColor, defaultValue=<rgba 0.25 0.25 0.25 1>}.RGB
    B=Multiply(A=@n1.RGB, B=@n3.Result)
    Alpha=LinearInterpolate(A=@n8.Result, B=@n9, Alpha=@n10)
   )
  ) {constAlpha=1}
  LinearInterpolate(A=@n11, B=Divide(A=@n11, B=EyeAdaptation().EyeAdaptation))
  n12 = WorldPosition()
  n13 = TransformPosition(None=AppendVector(A=ScreenPosition().ViewportUV, B=Constant())) {transformSourceType=TRANSFORMPOSSOURCE_View, transformType=TRANSFORMPOSSOURCE_World}
  Multiply(A=Normalize(VectorInput=Subtract(A=@n12.XYZ, B=@n13)), B=ScalarParameter() {defaultValue=-20, parameterName=OffsetToViewPos})
  n14 = Constant()
  Multiply(A=Constant() {r=1}, B=If(A=SceneDepth().R, B=PixelDepth().R, "A > B"=Constant() {r=1}, "A == B"=@n14, "A < B"=@n14))
  n15 = ComponentMask(None=VectorParameter() {group="3) Sphere Mask", parameterName=SphereMask Offset, defaultValue=<rgba 0 0 45 0>}.RGB) {b=true, g=true, r=true}
  SphereMask(
   A=WorldPosition().XYZ
   B=Add(A=ObjectPositionWS(), B=Transform(None=@n15) {transformSourceType=TRANSFORMSOURCE_Local})
   Radius=ScalarParameter() {defaultValue=52.8421, parameterName=Radius}
   Hardness=ScalarParameter() {defaultValue=0.5, parameterName=Hardness}
  )
 ]
}
