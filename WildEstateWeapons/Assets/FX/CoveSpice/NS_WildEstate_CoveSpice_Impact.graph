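# Niagara impact effect for the Wild Estate CoveSpice weapon: a brief glow with flickering light, sparks and water splashes.
# User.ChargePercent shifts glow, light and spark colors from yellow toward orange-red.
# User.PhysicalSurface gates spawn probability: SurfaceType8 gets droplets and a splash sprite, while sparks play everywhere except SurfaceType8 and SurfaceType30.
# System.CameraFade dims glow, light and sparks; longer-lived GPU sparks collide and lean against owner velocity.
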NiagaraSystem /WildEstateWeapons/Assets/FX/CoveSpice/NS_WildEstate_CoveSpice_Impact
dependencies {
 usedDataInterfaces [&NiagaraDataInterfaceColorCurve, &NiagaraDataInterfaceCurve, &NiagaraDataInterfaceCamera]
 usedDynamicInputs [
  &RemapRange
  &DistanceBetweenPositions
  &ReturnCameraPosition
 ]
 usedModules [
  &SystemState
  &self:SystemSpawnScript.SystemScriptSource.SystemScriptGraph.NiagaraNodeAssignment_1.SetVariables_#1
 ]
}
emitters {
 Corona {
  emitterUpdateScript [
   &EmitterState()
   &SpawnBurst_Instantaneous(Spawn Count=1, Use Spawn Probability=false)
   &SpawnRate(SpawnRate=6, Use Spawn Probability=false) {enabled=false}
  ]
  particleSpawnScript [
   &InitializeParticle(
    Color=&ColorFromCurve(
     ColorCurve=DataInterface() {propertyValues {
       AlphaCurve {keys [<at 0 1 RCIM_Cubic>]}
       BlueCurve {keys [<at 0 0.0375 RCIM_Cubic>,<at 1 0.0319 RCIM_Cubic>]}
       GreenCurve {keys [<at 0 0.9916 RCIM_Cubic>,<at 1 0.1864 RCIM_Cubic>]}
       RedCurve {keys [<at 0 1 RCIM_Cubic>,<at 1 1 RCIM_Cubic>]}
      }}
     CurveIndex=$User.ChargePercent
    )
    Lifetime=0.1
    Sprite Rotation Mode=::NewEnumerator2
    Sprite Size Mode=::NewEnumerator1
    Uniform Sprite Size Max=100
    Uniform Sprite Size Min=80
    UsePositionOffset=true
    Write Material Random=true
   )
   &DynamicMaterialParameters(
    Index 0 Param 1=0.35
    Index 0 Param 2=1
    Module.Param0WriteEnabled=true
    Module.Param1WriteEnabled=true
    Module.Param2WriteEnabled=false
    Module.Param3WriteEnabled=false
   )
  ]
  particleUpdateScript [
   &ParticleState()
   &CameraOffset(Camera Offset Amount=10)
   &ScaleColor(
    Scale Alpha=&FloatFromCurve(
     FloatCurve=DataInterface() {propertyValues {Curve {keys [
         {arriveTangent=2.3365, interpMode=RCIM_Cubic, leaveTangent=2.3365, tangentMode=RCTM_User, time=0, value=0}
         {interpMode=RCIM_Cubic, tangentMode=RCTM_User, time=0.5, value=1}
        ]}}}
     Scale Curve=&Lerp_Float(A=0.1, B=1, Alpha=$System.CameraFade)
    )
    ScaleA=true
    ScaleRGB=false
   )
   &ScaleSpriteSize(Scale Factor=&Vector2DFromFloat(Value=&FloatFromCurve(FloatCurve=DataInterface() {propertyValues {Curve {keys [
          {arriveTangent=1.3933, interpMode=RCIM_Cubic, leaveTangent=1.3933, tangentMode=RCTM_User, time=0, value=0.5}
          {interpMode=RCIM_Cubic, leaveTangent=0.353, time=1, value=1}
         ]}}})))
   &FN_Light_Attributes(
    Color=&ColorFromCurve(
     ColorCurve=DataInterface() {propertyValues {
       AlphaCurve {keys [<at 0 1 RCIM_Cubic>]}
       BlueCurve {keys [<at 0 0.0132 RCIM_Cubic>,<at 1 0 RCIM_Cubic>]}
       GreenCurve {keys [<at 0 1 RCIM_Cubic>,<at 1 0.1794 RCIM_Cubic>]}
       RedCurve {keys [<at 0 0.6958 RCIM_Cubic>,<at 1 1 RCIM_Cubic>]}
      }}
     CurveIndex=$User.ChargePercent
    )
    Create Color Attribute=true
    Create Exponent Attribute=true
    Create Radius Attribute=true
    Exponent=2
    Radius=125
    Scale Color=&Multiply_Float(
     A=&Waveform(Global Amplitude Scale=4, "Waveform [1]"=::NewEnumerator12, "[1] Frequency"=13, "[1] Amplitude Min/Max"=<0.33 1>)
     B=&Lerp_Float(A=0.5, B=1, Alpha=$System.CameraFade)
    )
   )
  ]
  rendererClasses [&NiagaraSpriteRendererProperties, &NiagaraLightRendererProperties]
  renderers [
   {rendererClass=&NiagaraSpriteRendererProperties}
   {rendererClass=&NiagaraLightRendererProperties, rendererIndex=1}
  ]
 }
 FN_Weapons_Debris_Sprite_SubUV_BETA003 {
  emitterUpdateScript [
   &EmitterState()
   &SpawnBurst_Instantaneous(
    Spawn Count=1
    Spawn Probability=&MakeFloatFromInt(Int=/Game/Effects/Niagara/DataChannels/SelectIntBySurface(
      Default:NiagaraInt32=0
      Surface:EPhysicalSurface=&self:ScratchDynamicInput_01(PhysicalSurface:NiagaraInt32=$User.PhysicalSurface)
      Num Surfaces:NiagaraInt32=1
      Surface 0:EPhysicalSurface=SurfaceType8
      Value For Surface 0:NiagaraInt32=1
      Surface 1:EPhysicalSurface=?
      Value For Surface 1:NiagaraInt32=?
      Surface 2:EPhysicalSurface=?
      Value For Surface 2:NiagaraInt32=?
      Surface 3:EPhysicalSurface=?
      Value For Surface 3:NiagaraInt32=?
     ))
    Use Spawn Probability=true
   )
  ]
  particleSpawnScript [
   /Niagara/Modules/Spawn/Initialization/InitializeParticle(
    Lifetime=0.3
    Module.Write Color=true
    Module.Write Lifetime=true
    Module.Write Mass=true
    Module.Write Material Random=true
    Module.Write Position=true
    Module.Write Scale=false
    Module.Write SpriteRotation=false
    Module.Write SpriteSize=true
    Position=&Add_Vector(A=/Niagara/DynamicInputs/Helpers/SimulationPosition(), B=<0 0 2>)
    Sprite Size=&RandomRangeVector2D(Minimum=<6 10>, Maximum=<12 16>)
   )
   &AddVelocityInCone(
    Cone Angle=30
    Module.Use Velocity Falloff On Cone Axis=false
    Velocity Strength=&UniformRangedFloat(Minimum=200, Maximum=300)
   )
   &FN_PerParticleLOD(Enable SpriteSize Distance Scaling=true, Sprite Size Scale Far=<1.5 1.5>)
   /NiagaraFN/Sprite/FN_RandomizeSpriteUV()
  ]
  particleUpdateScript [
   &ParticleState()
   &AccelerationForce(Acceleration=<0 0 -800>)
   &Drag(Drag=0.1, Use Linear Drag=true, Use Rotational Drag=false) {enabled=false}
   &SolveForcesAndVelocity(Clamp Velocity=false, Limit Acceleration=false)
   &ScaleSpriteSize(Scale Factor=&Vector2DFromFloat(Value=&FloatFromCurve(FloatCurve=DataInterface() {propertyValues {Curve {keys [<at -0.0046 1.0113>,<at 0.77 7.78>,<at 1.0011 0.002>]}}})))
   &DynamicMaterialParameters(
    Index 0 Param 2=&FloatFromCurve(FloatCurve=DataInterface() {propertyValues {Curve {keys [
         {arriveTangent=46.3377, interpMode=RCIM_Cubic, leaveTangent=46.3379, tangentMode=RCTM_User, time=-0.0046, value=0}
         {arriveTangent=0.0019, interpMode=RCIM_Cubic, leaveTangent=0.0019, time=0.05, value=0.9763}
         {arriveTangent=-2.5858, interpMode=RCIM_Cubic, leaveTangent=-2.5858, tangentMode=RCTM_User, time=1.0011, value=0.002}
        ]}}})
    Module.Param0WriteEnabled=false
    Module.Param1WriteEnabled=true
    Module.Param2WriteEnabled=false
    Module.Param3WriteEnabled=false
   )
   /Niagara/Modules/Update/SubUV/SubUVAnimation(Random SubUV Lookup Index:NiagaraFloat=$Particles.Initial.MaterialRandom, SubUV Animation Mode=::NewEnumerator1) {enabled=false}
   &Color(Color=&MakeLinearColorFromVectorAndFloat(
     "Vector (RGB)"=<0.8 1 1>
     "Float (Alpha)"=&FloatFromCurve(FloatCurve=DataInterface() {propertyValues {Curve {keys [<at 0.4 1.0113 RCIM_Cubic>,<at 1.0011 0.002>]}}})
    ))
  ]
  rendererClasses [&NiagaraSpriteRendererProperties]
  renderers [{rendererClass=&NiagaraSpriteRendererProperties}]
 }
 SparksLazy001 {
  emitterUpdateScript [
   &EmitterState()
   &SpawnBurst_Instantaneous(
    Spawn Count=&RandomRangeInt(Minimum=2, Maximum=3)
    Spawn Probability=&MakeFloatFromInt(Int=&SelectIntBySurface(
      Default:NiagaraInt32=1
      Surface:EPhysicalSurface=&self:ScratchDynamicInput_01(PhysicalSurface:NiagaraInt32=$User.PhysicalSurface)
      Num Surfaces:NiagaraInt32=2
      Surface 0:EPhysicalSurface=SurfaceType8
      Value For Surface 0:NiagaraInt32=0
      Surface 1:EPhysicalSurface=SurfaceType30
      Value For Surface 1:NiagaraInt32=0
      Surface 2:EPhysicalSurface=?
      Value For Surface 2:NiagaraInt32=?
      Surface 3:EPhysicalSurface=?
      Value For Surface 3:NiagaraInt32=?
     ))
    Use Spawn Probability=true
    UseLoopCountLimit=false
   )
  ]
  particleSpawnScript [
   /Niagara/Modules/Spawn/Initialization/V2/InitializeParticle(
    AdjustAlpha=false
    AdjustHue=false
    AdjustSaturation=false
    AdjustValue=true
    Color=&ColorFromCurve(
     ColorCurve=DataInterface() {propertyValues {
       AlphaCurve {keys [<at 0 1 RCIM_Cubic>]}
       BlueCurve {keys [<at 0.0344 0.0673 RCIM_Cubic>,<at 1 0.0319 RCIM_Cubic>]}
       GreenCurve {keys [<at 0.0344 0.8946 RCIM_Cubic>,<at 1 0.1864 RCIM_Cubic>]}
       RedCurve {keys [<at 0.0344 1 RCIM_Cubic>,<at 1 1 RCIM_Cubic>]}
      }}
     CurveIndex=$User.ChargePercent
    )
    Color Mode=::NewEnumerator3
    Lifetime Max=0.2
    Lifetime Min=0.1
    Lifetime Mode=::NewEnumerator1
    Sprite Size Max=<4 15>
    Sprite Size Min=<3 5>
    Sprite Size Mode=::NewEnumerator2
    UsePositionOffset=false
    Value Range:Vector2f=<0.5 5>
    Write Material Random=true
   )
   &SphereLocation(
    Hemisphere X=true
    MaskSpawnGroup=false
    Override Seed=false
    Sphere Radius=5
    Surface Only=true
    Surface Only Band Thickness=0.1
    UseNonUniformScale=false
    UseOffset=false
    UseSphereOrientation=false
   ) {module=SphereLocation001}
   &AddVelocity(Velocity=&Multiply_VectorByFloat(
     Vector=$Output.SphereLocation001.SphereNormal
     Float=&RandomRangeFloat(Minimum=400, Maximum=800, Evaluation Type=::NewEnumerator1)
    ))
  ]
  particleUpdateScript [
   &ParticleState()
   &AccelerationForce(Acceleration=&RandomRangeVector(Minimum=<0 0 -1000>, Maximum=<0 0 1000>))
   &Drag(Drag=&RandomRangeFloat(Minimum=1, Maximum=4), Use Linear Drag=true, Use Rotational Drag=false)
   &SolveForcesAndVelocity(Clamp Velocity=false, Limit Acceleration=false)
   &ScaleSpriteSizeBySpeed(
    Max Scale Factor=<0.75 6>
    Min Scale Factor=<1 1.5>
    Source Velocity:Vector3f=&Multiply_Vector(A=<0.01 0.01 1>, B=$Particles.Velocity)
   )
   &ScaleColor(
    Scale Alpha=&FloatFromCurve(
     FloatCurve=DataInterface() {propertyValues {Curve {keys [<at 0 3>,<at 0.2 1>]}}}
     Scale Curve=&Lerp_Float(A=0.1, B=1, Alpha=$System.CameraFade)
    )
    Scale RGB=&MakeVectorFromLinearColorRGB(Color=&ColorFromCurve(ColorCurve=DataInterface() {propertyValues {
        AlphaCurve {keys [<at 0 1 RCIM_Cubic>,<at 1 1 RCIM_Cubic>]}
        BlueCurve {keys [<at 0 8.0507 RCIM_Cubic>,<at 0.4236 0>]}
        GreenCurve {keys [<at 0 17.6367 RCIM_Cubic>,<at 0.4236 7.6165>]}
        RedCurve {keys [<at 0 25 RCIM_Cubic>,<at 0.4236 20>]}
       }}))
    ScaleA=true
    ScaleRGB=false
   )
   &ScaleSpriteSize(Scale Factor=&Vector2DFromFloat(Value=&FloatFromCurve(FloatCurve=DataInterface() {propertyValues {Curve {keys [
          {tangentMode=RCTM_User, time=0, value=1}
          {arriveTangent=-0.0405, interpMode=RCIM_Cubic, leaveTangent=-0.0405, tangentMode=RCTM_User, time=0.7009, value=1}
          {arriveTangent=-8.0089, interpMode=RCIM_Cubic, leaveTangent=-8.0089, tangentMode=RCTM_User, time=1, value=0}
         ]}}})))
  ]
  rendererClasses [&NiagaraSpriteRendererProperties]
  renderers [{rendererClass=&NiagaraSpriteRendererProperties}]
 }
 SparksLazy002 {
  simTarget=GPUComputeSim
  emitterUpdateScript [
   &EmitterState()
   &SpawnRate(
    Spawn Probability=&MakeFloatFromInt(Int=&SelectIntBySurface(
      Default:NiagaraInt32=1
      Surface:EPhysicalSurface=&self:ScratchDynamicInput_01(PhysicalSurface:NiagaraInt32=$User.PhysicalSurface)
      Num Surfaces:NiagaraInt32=2
      Surface 0:EPhysicalSurface=SurfaceType8
      Value For Surface 0:NiagaraInt32=0
      Surface 1:EPhysicalSurface=SurfaceType30
      Value For Surface 1:NiagaraInt32=0
      Surface 2:EPhysicalSurface=?
      Value For Surface 2:NiagaraInt32=?
      Surface 3:EPhysicalSurface=?
      Value For Surface 3:NiagaraInt32=?
     ))
    SpawnRate=30
    Use Spawn Probability=true
   )
   &SpawnBurst_Instantaneous(
    Spawn Count=&RandomRangeInt(Minimum=6, Maximum=15)
    Spawn Probability=&MakeFloatFromInt(Int=&SelectIntBySurface(
      Default:NiagaraInt32=1
      Surface:EPhysicalSurface=&self:ScratchDynamicInput_01(PhysicalSurface:NiagaraInt32=$User.PhysicalSurface)
      Num Surfaces:NiagaraInt32=2
      Surface 0:EPhysicalSurface=SurfaceType8
      Value For Surface 0:NiagaraInt32=0
      Surface 1:EPhysicalSurface=SurfaceType30
      Value For Surface 1:NiagaraInt32=0
      Surface 2:EPhysicalSurface=?
      Value For Surface 2:NiagaraInt32=?
      Surface 3:EPhysicalSurface=?
      Value For Surface 3:NiagaraInt32=?
     ))
    Use Spawn Probability=true
    UseLoopCountLimit=false
   )
  ]
  particleSpawnScript [
   &InitializeParticle(
    AdjustAlpha=false
    AdjustHue=false
    AdjustSaturation=false
    AdjustValue=true
    Color=&ColorFromCurve(
     ColorCurve=DataInterface() {propertyValues {
       AlphaCurve {keys [<at 0 1 RCIM_Cubic>]}
       BlueCurve {keys [<at 0 0.034 RCIM_Cubic>,<at 1 0.0319 RCIM_Cubic>]}
       GreenCurve {keys [<at 0 0.9412 RCIM_Cubic>,<at 1 0.1864 RCIM_Cubic>]}
       RedCurve {keys [<at 0 1 RCIM_Cubic>,<at 1 1 RCIM_Cubic>]}
      }}
     CurveIndex=$User.ChargePercent
    )
    Color Mode=::NewEnumerator3
    Lifetime Max=1
    Lifetime Min=0.8
    Lifetime Mode=::NewEnumerator1
    Sprite Size Mode=::NewEnumerator1
    Uniform Sprite Size Max=4
    Uniform Sprite Size Min=2
    UsePositionOffset=false
    Value Range:Vector2f=<0.5 5>
    Write Material Random=true
   ) {module=InitializeParticle001}
   &SphereLocation(
    Hemisphere X=true
    MaskSpawnGroup=false
    Override Seed=false
    Sphere Radius=5
    Surface Only=false
    UseNonUniformScale=false
    UseOffset=false
    UseSphereOrientation=false
   ) {module=SphereLocation001}
   &AddVelocity(Velocity=&Multiply_VectorByFloat(
     Vector=&NormalizeVector(
      Vector To Normalize=&Add_Vector(
       A=<0.6 0 0>
       B=&TransformVector(
        Vector=&NormalizeVector(
         Vector To Normalize=/Niagara/DynamicInputs/Vectors/NegateVector(Vector:Vector3f=$Engine.Owner.Velocity)
         Safe Normalize=true
         Safe Fallback Threshold=0
         Safe Fallback Vector=<0 0 0.001>
        )
        Source Space=::World
        Destination Space=::Local
       )
      )
      Safe Normalize=true
      Safe Fallback Threshold=0
      Safe Fallback Vector=<0 0 0.001>
     )
     Float=&RandomRangeFloat(Minimum=50, Maximum=300)
    ))
   &AddVelocity(
    Coordinate Space=::Simulation
    Velocity=&Multiply_VectorByFloat(
     Vector=$Output.SphereLocation001.SphereNormal
     Float=&RandomRangeFloat(Minimum=50, Maximum=500, Evaluation Type=::NewEnumerator1)
    )
   ) {module=AddVelocity001}
  ]
  particleUpdateScript [
   &ParticleState()
   &AccelerationForce(Acceleration=<0 0 -900>)
   &Drag(Drag=2, Use Linear Drag=true, Use Rotational Drag=false)
   &Collision(
    Age Colliding Particles=false
    Kill Offscreen Particles=true
    Module.Control Rest State Via Surface Normal=false
    Module.EnableMaxCollisionCount=false
    Module.Randomize Collision Normal Vector=false
    Restitution=0.1
   )
   &SolveForcesAndVelocity(Clamp Velocity=false, Limit Acceleration=false)
   &ScaleSpriteSizeBySpeed(
    Max Scale Factor=<0.66 8>
    Source Velocity:Vector3f=&Multiply_Vector(A=<0.01 0.01 1>, B=$Particles.Velocity)
    Velocity Threshold=600
   )
   &ScaleColor(
    Scale Alpha=&FloatFromCurve(
     FloatCurve=DataInterface() {propertyValues {Curve {keys [
         {tangentMode=RCTM_User, time=0, value=1}
         {arriveTangent=-0.0405, interpMode=RCIM_Cubic, leaveTangent=-0.0405, tangentMode=RCTM_User, time=0.7009, value=1}
         {arriveTangent=-8.0089, interpMode=RCIM_Cubic, leaveTangent=-8.0089, tangentMode=RCTM_User, time=1, value=0}
        ]}}}
     Scale Curve=&Lerp_Float(A=0.1, B=1, Alpha=$System.CameraFade)
    )
    Scale RGB=&MakeVectorFromLinearColorRGB(Color=&ColorFromCurve(ColorCurve=DataInterface() {propertyValues {
        AlphaCurve {keys [<at 0 1 RCIM_Cubic>,<at 1 1 RCIM_Cubic>]}
        BlueCurve {keys [{arriveTangent=-1.3944, interpMode=RCIM_Cubic, time=0, value=4}, <at 0.4758 1.5>, <at 0.9516 0.66>]}
        GreenCurve {keys [{arriveTangent=-11.4062, interpMode=RCIM_Cubic, time=0, value=4}, <at 0.4758 1.5>, <at 0.9516 0.66>]}
        RedCurve {keys [{arriveTangent=-20.6532, interpMode=RCIM_Cubic, time=0, value=4}, <at 0.4758 1.5>, <at 0.9516 0.66>]}
       }}))
    ScaleA=true
    ScaleRGB=true
   )
   &ScaleSpriteSize(Scale Factor=&Vector2DFromFloat(Value=&FloatFromCurve(
      FloatCurve=DataInterface() {propertyValues {Curve {keys [
          {tangentMode=RCTM_User, time=0, value=1}
          {arriveTangent=-0.0405, interpMode=RCIM_Cubic, leaveTangent=-0.0405, tangentMode=RCTM_User, time=0.7009, value=1}
          {arriveTangent=-8.0089, interpMode=RCIM_Cubic, leaveTangent=-8.0089, tangentMode=RCTM_User, time=1, value=0}
         ]}}}
      Scale Curve=0.75
     )))
  ]
  rendererClasses [&NiagaraSpriteRendererProperties]
  renderers [{rendererClass=&NiagaraSpriteRendererProperties}]
 }
 Water_Drops {
  simTarget=GPUComputeSim
  emitterUpdateScript [
   &EmitterState(
    Enable Distance Culling=true
    Max Distance Response=::KillAfterParticlesFinish
    Module.bUseMaxDistance=true
    Module.bUseMinDistance=false
    Scalability Mode=::NewEnumerator1
   )
   &SpawnBurst_Instantaneous(
    Spawn Count=&UniformRangedInt(Minimum=15, Maximum=30)
    Spawn Probability=&MakeFloatFromInt(Int=&SelectIntBySurface(
      Default:NiagaraInt32=0
      Surface:EPhysicalSurface=&self:ScratchDynamicInput_01(PhysicalSurface:NiagaraInt32=$User.PhysicalSurface)
      Num Surfaces:NiagaraInt32=1
      Surface 0:EPhysicalSurface=SurfaceType8
      Value For Surface 0:NiagaraInt32=1
      Surface 1:EPhysicalSurface=?
      Value For Surface 1:NiagaraInt32=?
      Surface 2:EPhysicalSurface=?
      Value For Surface 2:NiagaraInt32=?
      Surface 3:EPhysicalSurface=?
      Value For Surface 3:NiagaraInt32=?
     ))
    Use Spawn Probability=true
   )
  ]
  particleSpawnScript [
   /Niagara/Modules/Spawn/Initialization/InitializeParticle(
    Lifetime=&UniformRangedFloat(Minimum=0.3, Maximum=1)
    Module.Write Color=true
    Module.Write Lifetime=true
    Module.Write Mass=true
    Module.Write Material Random=true
    Module.Write Position=false
    Module.Write Scale=false
    Module.Write SpriteRotation=false
    Module.Write SpriteSize=true
    Position=/Niagara/Functions/Localspace/SimulationPosition()
    Sprite Size=&UniformRangedVector2D(Minimum=<3 4>, Maximum=<7 8>)
   )
   &AddVelocityInCone(
    Cone Angle=90
    Module.Use Velocity Falloff On Cone Axis=false
    Velocity Strength=&UniformRangedFloat(Minimum=100, Maximum=400)
   )
   &FN_PerParticleLOD(Enable SpriteSize Distance Scaling=true, Sprite Size Scale Far=<1.5 1.5>)
  ]
  particleUpdateScript [
   &ParticleState()
   &AccelerationForce(Acceleration=<0 0 -980>)
   &Drag(Drag=0.1, Use Linear Drag=true, Use Rotational Drag=false) {enabled=false}
   &SolveForcesAndVelocity(Clamp Velocity=false, Limit Acceleration=false)
   &ScaleSpriteSize(Scale Factor=&Vector2DFromFloat(Value=&OneMinusFloat(Float=$Particles.NormalizedAge)))
   &ScaleSpriteSizeBySpeed(Max Scale Factor=<1 2>, Velocity Threshold=500)
   &SubUVAnimation(Random SubUV Lookup Index:NiagaraFloat=$Particles.Initial.MaterialRandom, SubUV Animation Mode=::NewEnumerator1) {enabled=false}
   &Color(Color=&MakeLinearColorFromVectorAndFloat(
     "Vector (RGB)"=<0.8 1 1>
     "Float (Alpha)"=&FloatFromCurve(FloatCurve=DataInterface() {propertyValues {Curve {keys [<at 0.4 1.0113 RCIM_Cubic>,<at 1.0011 0.002>]}}})
    ))
  ]
  rendererClasses [&NiagaraSpriteRendererProperties]
  renderers [{rendererClass=&NiagaraSpriteRendererProperties}]
 }
}
summary {
 systemName=NS_WildEstate_CoveSpice_Impact
 emitters [
  {emitterName=SparksLazy001, rendererClasses [&NiagaraSpriteRendererProperties]}
  {emitterName=SparksLazy002, simTarget=GPUComputeSim, rendererClasses [&NiagaraSpriteRendererProperties]}
  {emitterName=Corona, rendererClasses [&NiagaraSpriteRendererProperties, &NiagaraLightRendererProperties]}
  {emitterName=Water_Drops, simTarget=GPUComputeSim, rendererClasses [&NiagaraSpriteRendererProperties]}
  {emitterName=FN_Weapons_Debris_Sprite_SubUV_BETA003, rendererClasses [&NiagaraSpriteRendererProperties]}
 ]
 userVariables [
  User.BeamEnd_0 {type=NiagaraPosition, description="", defaultValue {}}
  User.bIsMudImpact {type=NiagaraBool}
  User.bIsWaterImpact {type=NiagaraBool}
  User.ChargePercent {type=NiagaraFloat, defaultValue=0, description=""}
  User.PhysicalSurface {type=NiagaraInt32}
  User.RibbonID {type=NiagaraInt32}
  User.WaterBodyIndex {type=NiagaraFloat, defaultValue=0, description=""}
  User.WaterDepth {type=NiagaraFloat, defaultValue=0, description=""}
  User.WaveOffset {type=Vector3f, description="", defaultValue {struct=Vector3f, value=<0 0 0>}}
 ]
}
