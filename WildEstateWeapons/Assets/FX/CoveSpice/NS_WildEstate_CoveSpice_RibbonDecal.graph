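# Niagara ribbon trail for the CoveSpice weapon effects, apparently a decal-like streak laid along a surface as the system moves.
# Particles spawn per unit travelled, nudged along local X, with ribbon facing taken from the system's X axis.
# Short-lived segments taper in width, fade out and darken from a hot flash, tinted yellow-orange to red by User.ChargePercent.
# Mud, water and physical-surface user parameters are exposed but unreferenced in the emitter shown.
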NiagaraSystem /WildEstateWeapons/Assets/FX/CoveSpice/NS_WildEstate_CoveSpice_RibbonDecal
dependencies {
 usedDataInterfaces [&NiagaraDataInterfaceColorCurve, &NiagaraDataInterfaceCurve]
 usedDynamicInputs []
 usedModules [&SystemState]
}
emitters {Ribbon_Trail {
  emitterUpdateScript [
   &EmitterState()
   &SpawnPerUnit(Module.Use Max Movement Threshold=true, Module.Use Movement Tolerance=false, Spawn Spacing=5, Use Spawn Probability=false)
  ]
  particleSpawnScript [&InitializeParticle(
    Lifetime=0.7
    Position Offset=<0.5 0 0>
    Position Offset Coordinate Space=::Local
    Ribbon Facing Mode=::NewEnumerator1
    Ribbon Facing Vector:Vector3f=$Engine.Owner.SystemXAxis
    Ribbon Width=30
    Ribbon Width Mode=::NewEnumerator1
    Sprite Size Mode=::NewEnumerator3
    Uniform Sprite Size=20
    UsePositionOffset=true
    Write Material Random=true
   )]
  particleUpdateScript [
   &ParticleState()
   &ScaleColor(
    Scale Alpha=&FloatFromCurve(
     FloatCurve=DataInterface() {propertyValues {Curve {keys [{interpMode=RCIM_Cubic, tangentMode=RCTM_User, time=0.3, value=1}, {interpMode=RCIM_Cubic, leaveTangent=-7.011, time=1, value=0}]}}}
     Scale Curve=0.7
    )
    Scale RGB=&MakeVectorFromLinearColorRGB(Color=&MultiplyLinearColor(
      Linear Color=&LerpLinearColors(StartColor=<rgba 4 2.4563 0 1>, EndColor=<rgba 4 0.152 0.028 1>, LerpFactor=$User.ChargePercent)
      Scale Factor=&MakeVector4_FromLinearColor(LinearColor=&ColorFromCurve(ColorCurve=DataInterface() {propertyValues {
          AlphaCurve {keys [<at 0 1 RCIM_Cubic>,<at 1 0 RCIM_Cubic>]}
          BlueCurve {keys [<at 0.0293 4 RCIM_Cubic>,<at 0.1778 1>,<at 0.3823 0.099>,<at 0.6103 0.01 RCIM_Cubic>]}
          GreenCurve {keys [<at 0.0293 4 RCIM_Cubic>,<at 0.1778 1>,<at 0.3823 0.099>,<at 0.6103 0.01 RCIM_Cubic>]}
          RedCurve {keys [<at 0.0293 4 RCIM_Cubic>,<at 0.1778 1>,<at 0.3823 0.099>,<at 0.6103 0.01 RCIM_Cubic>]}
         }}))
     ))
    ScaleA=true
    ScaleRGB=true
   )
   &ScaleRibbonWidth(Ribbon Width Scale=&FloatFromCurve(FloatCurve=DataInterface() {propertyValues {Curve {
        defaultValue=1
        keys [
         {arriveTangent=6.1286, interpMode=RCIM_Cubic, leaveTangent=6.1286, tangentMode=RCTM_User, time=0, value=0}
         {arriveTangent=0.0224, leaveTangent=0.0224, time=0.1, value=1}
         {arriveTangent=-1.1111, interpMode=RCIM_Cubic, leaveTangent=-1.1111, time=0.7, value=1}
         {arriveTangent=-11.4405, interpMode=RCIM_Cubic, leaveTangent=-11.4405, tangentMode=RCTM_User, time=1, value=0}
        ]
       }}}))
  ]
  rendererClasses [&NiagaraRibbonRendererProperties]
  renderers [{rendererClass=&NiagaraRibbonRendererProperties}]
 }}
summary {
 systemName=NS_WildEstate_CoveSpice_RibbonDecal
 emitters [{emitterName=Ribbon_Trail, rendererClasses [&NiagaraRibbonRendererProperties]}]
 userVariables [
  User.bIsMudImpact {type=NiagaraBool}
  User.bIsWaterImpact {type=NiagaraBool}
  User.ChargePercent {type=NiagaraFloat, defaultValue=0, description=""}
  User.PhysicalSurface {type=NiagaraInt32}
 ]
}
